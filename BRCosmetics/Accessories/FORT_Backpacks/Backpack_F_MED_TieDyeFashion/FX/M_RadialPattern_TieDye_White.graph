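# Translucent, lit, two-sided Niagara material that draws an animated radial tie-dye swirl in a white variant, apparently for the TieDyeFashion backpack FX.
# Rotating UVs become distance and angle, distorted by curve offsets, panned noise and spiky seams into scrolling spiral bands.
# Curve atlases blend three tones from dynamic parameter red channels, with noisy white edges pin-light blended on top.
# A radial mask curve with depth fade shapes the alpha.
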Material /BRCosmetics/Accessories/FORT_Backpacks/Backpack_F_MED_TieDyeFashion/FX/M_RadialPattern_TieDye_White
graph {
 properties {
  bCastRayTracedShadows=false
  bUsedWithNiagaraMeshParticles=true
  bUsedWithNiagaraSprites=true
  blendMode=BLEND_Translucent
  pixelDepthOffsetMode=PDOM_Legacy
  translucencyLightingMode=TLM_VolumetricDirectional
  twoSided=true
  usedShadingModels=Default Lit
 }
 nodes [
  Constant() {r=1}
  Constant() {r=0.8}
  n1 = DynamicParameter() {parameterIndex=1, defaultValue=<rgba 0.0991 0.068 1 1>}
  n2 = DynamicParameter() {parameterIndex=2, defaultValue=<rgba 1 0 0.8096 1>}
  n3 = MaterialFunctionCall(X=@n2.R, Y=@n2.R, Z=@n2.R, A=@n2.R) {materialFunction=&MakeFloat4}
  n4 = DynamicParameter() {parameterIndex=3, defaultValue=<rgba 0.1249 0.9901 1 1>}
  n5 = Rotator(
   Coordinate=TextureCoordinate()
   Time=Multiply(A=Time(), B=ScalarParameter() {defaultValue=0.3, parameterName=RotationSpeed, sliderMax=1}) {constB=0.3}
  )
  n6 = MaterialFunctionCall(Float2=@n5) {materialFunction=&BreakOutFloat2Components}
  n7 = Subtract(A=AppendVector(A=OneMinus(None=@n6.G), B=@n6.R)) {constB=0.5}
  n8 = SquareRoot(None=Multiply(
    A=MaterialFunctionCall(Vector 2=@n7) {materialFunction=/Engine/Functions/Engine_MaterialFunctions02/Utility/VectorLength}.V2 Length
    B=Constant() {r=0.25}
   ))
  n9 = Add(A=@n8, B=Constant())
  n10 = ScalarParameter() {defaultValue=0.6, parameterName=RadialScale}
  n11 = MaterialFunctionCall(Float2=@n7) {materialFunction=&BreakOutFloat2Components}
  n12 = Divide(
   A=Add(
    A=Arctangent2Fast(Y=@n11.G, X=@n11.G)
    B=MaterialFunctionCall() {materialFunction=&Pi}.Result
   )
   B=MaterialFunctionCall(Multiplier=Constant() {r=2}) {materialFunction=&Pi}.Result
  )
  n13 = Add(A=@n9, B=@n12)
  n14 = CurveAtlasRowParameter(CurveTime=Fmod(A=Multiply(A=@n13) {constB=3}, B=Constant() {r=1})) {
   atlas=&/CA_TieDye
   curve=&/C_RadialOffset_StarSharp
   parameterName=CurveAtlas_RadialOffset
   sortPriority=44
  }
  n15 = Constant() {r=0.0005}
  n16 = Time()
  n17 = TextureCoordinate()
  n18 = TextureSampleParameter2D!(
   UVs=Panner(Coordinate=AppendVector(A=@n8, B=@n12), Time=Multiply(A=@n16) {constB=0.04, desc=mask rotate speed}) {speedX=1.5, speedY=1.5}
   "DDX(UVs)"=DDX(Value=@n17)
   "DDY(UVs)"=DDY(Value=@n17)
   Apply View MipBias
  ) {mipValueMode=TMVM_Derivative, parameterName=Param, texture=/Game/Global/Textures/Default/Noise/TankMoteNoise}
  n19 = Constant()
  n20 = Constant() {r=1}
  n21 = Constant() {r=0.3}
  n22 = ScalarParameter() {defaultValue=2, parameterName=Spike Frequency}
  n23 = Noise!(World Position=Add(A=Multiply(A=@n12, B=@n22) {constB=2}, B=Constant() {r=125}), FilterWidth) {levelScale=3, levels=2, outputMin=0, scale=3}
  n24 = TextureSample(UVs=Rotator(Time=Multiply(A=@n16) {constB=0.065, desc=spike rotate speed}) {speed=2}) {samplerType=SAMPLERTYPE_LinearGrayscale, texture=&/T_Radial_Seam_Patch}
  n25 = Multiply(A=Add(A=@n23, B=Multiply(A=@n24.R) {constB=2})) {constB=0.75}
  n26 = Multiply(
   A=Multiply(
    A=MaterialFunctionCall(Input=@n18.R, Input Low=@n19, Input High=@n20, Target Low=@n21, Target High=@n20) {materialFunction=&RemapValueRange}.Result
    B=@n25
   )
   B=ScalarParameter() {defaultValue=0.03, parameterName=Noise_Magnitude, sliderMax=1}
  ) {constB=0.01}
  n27 = Add(A=Add(A=Power(Base=@n9, Exp=@n10) {constExponent=0.65}, B=Multiply(A=@n14.R, B=@n15)), B=@n26)
  n28 = ScalarParameter() {defaultValue=2.5, parameterName=Repeat_Frequency, sliderMax=5}
  n29 = Multiply(A=Time() {bIgnorePause=true, period=1}, B=ScalarParameter() {defaultValue=0.1, parameterName=RadialSpeed, sliderMax=1, sortPriority=2}) {constB=0.1}
  n30 = Subtract(A=Multiply(A=Max(A=@n27) {constB=0.1}, B=@n28) {constB=2.5}, B=@n29)
  n31 = Constant() {r=1}
  n32 = Fmod(A=Add(A=Abs(None=@n30), B=@n12), B=@n31)
  n33 = CurveAtlasRowParameter(CurveTime=@n32) {
   atlas=&CA_TieDye
   curve=&/C_TwoTone_Grayscale
   defaultValue=4
   parameterName=CurveAtlas_Color
   sortPriority=31
  }
  n34 = Multiply(
   A=MaterialFunctionCall(
    A=MaterialFunctionCall(X=@n1.R, Y=@n1.R, Z=@n1.R, A=@n1.R) {materialFunction=&MakeFloat4}.Result
    B=@n3.Result
    C=MaterialFunctionCall(X=@n4.R, Y=@n4.R, Z=@n4.R, A=@n4.R) {materialFunction=&MakeFloat4}.Result
    Alpha=MaterialFunctionCall(Float3=@n33) {materialFunction=&BreakOutFloat3Components}.R
   ) {materialFunction=&Lerp_3Color}.Result
   B=ScalarParameter() {defaultValue=0.4, parameterName=ScaleColor, sliderMax=15}
  )
  n35 = LinearInterpolate(A=@n27, B=@n9) {constAlpha=1}
  n36 = Subtract(A=Multiply(A=Max(A=@n35) {constB=0.1}, B=@n28) {constB=2.5}, B=@n29)
  n37 = Fmod(A=Add(A=Abs(None=@n36), B=Multiply(B=@n26) {constA=3, constB=5, desc=num repeats}), B=Constant() {r=1})
  n38 = CurveAtlasRowParameter(CurveTime=LinearInterpolate(A=Add(A=@n32, B=@n37), B=@n32)) {
   atlas=&/CA_TieDye_White
   curve=&/C_WhiteEdges_B
   defaultValue=2
   parameterName=CA_Color_white
  }
  n39 = LinearInterpolate(A=@n25, B=@n26) {constAlpha=0.7}
  n40 = Noise(World Position=AppendVector(A=AppendVector(A=@n36, B=@n39), B=Constant())) {levelScale=3, levels=3, scale=0.4}
  n41 = Add(A=@n38, B=Clamp(None=Multiply(A=@n40) {constB=0.75}))
  n42 = Multiply(A=OneMinus(None=@n34), B=Multiply(A=Power(Base=@n41) {constExponent=3}, B=ScalarParameter() {defaultValue=1, parameterName=WhiteScale}))
  n43 = LinearInterpolate(
   A=Add(
    A=MaterialFunctionCall!(Base=@n34, Blend=@n42) {materialFunction=/Engine/Functions/Engine_MaterialFunctions03/Blends/Blend_PinLight}.Result
    B=@n34
   )
   B=Multiply(A=ComponentMask(None=@n3.Result) {b=true, g=true, r=true}) {constB=0.5}
   Alpha=SphereMask(A=TextureCoordinate(), B=Constant2Vector() {g=0.5, r=0.5}) {attenuationRadius=0.035, hardnessPercent=35}
  )
  n44 = CurveAtlasRowParameter(CurveTime=Fmod(A=@n27, B=@n31)) {
   atlas=&CA_TieDye
   curve=&/C_TieDie_Mask
   defaultValue=2
   parameterName=CurveAtlas_Mask
   sortPriority=31
  }
  n45 = DepthFade(Opacity=Multiply(A=@n33.A, B=@n44), FadeDistance=DynamicParameter() {parameterIndex=3, defaultValue=<rgba 0.1249 0.9901 1 1>}.R)
  n46 = Clamp(None=Subtract(A=Power(Base=@n45) {constExponent=1.5}) {constB=0.001})
  Divide(A=@n43, B=@n46)
  MaterialFunctionCall(Emissive=Multiply(A=@n43) {constB=0.1}) {materialFunction=&MF_FX_EmissiveAdjustments}
  Max(A=@n23, B=@n24.R)
  Multiply(A=@n46) {constB=10000}
  Multiply(A=@n12) {constB=0.2, desc=num swirls in tenths}
  n47 = AppendVector(A=@n5, B=Multiply(A=Time()) {constB=0.05})
  Multiply(A=Noise(World Position=Multiply(A=@n47)) {bTiling=true, levels=3, noiseFunction=NOISEFUNCTION_GradientTex, outputMin=0}) {constB=0.02}
  Reroute(None=@n25)
  TextureCoordinate()
  VectorParameter() {parameterName=Color_A, defaultValue=<rgba 0 0.0923 1 0>}
  VectorParameter() {parameterName=Color_B, defaultValue=<rgba 1 0 0.8995 0>}
  VectorParameter() {parameterName=Color_C, defaultValue=<rgba 0 1 0.4573 0>}
 ]
}
