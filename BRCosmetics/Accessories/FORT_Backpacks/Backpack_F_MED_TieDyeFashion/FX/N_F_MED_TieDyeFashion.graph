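# Niagara effect for the TieDyeFashion back bling: a spinning, tie-dye colored visual attached to the backpack mesh.
# TieDye_Base bursts one particle that follows a sampled socket on User.SkelMesh, with a scratch module applying offset and time-driven sprite rotation.
# Another scratch module passes three user colors into dynamic material parameters for two sprite and two mesh renderers, including a scaled halo size.
# A second torus-spawn emitter, Empty, is disabled.
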NiagaraSystem /BRCosmetics/Accessories/FORT_Backpacks/Backpack_F_MED_TieDyeFashion/FX/N_F_MED_TieDyeFashion
dependencies {usedDataInterfaces [], usedDynamicInputs [], usedModules [&SystemState]}
emitters {
 Empty {
  bEnabled=false
  emitterUpdateScript [
   &EmitterState()
   &SpawnRate(SpawnRate=25, Use Spawn Probability=false)
  ]
  particleSpawnScript [
   &InitializeParticle(
    Color Mode=::NewEnumerator0
    Lifetime=5
    Sprite Size Mode=::NewEnumerator3
    Uniform Sprite Size=2
    UsePositionOffset=false
    Write Material Random=true
   )
   &TorusLocation(Align Torus With Axis=true, Handle Radius=1, Large Radius=2, MaskSpawnGroup=false, Module.Override Randomness=false, Torus Axis=<1 0 0>)
   &self:ScratchModule(VelScale=5) {module=ScratchModule}
  ]
  particleUpdateScript [
   &ParticleState()
   &SkeletalMeshLocation(
    "Interpolate Mesh Sampling (Update)"=true
    MaskSpawnGroup=false
    Mesh Sampling Type=::NewEnumerator5
    Sampled Position Offset Coordinate Space:ENiagaraSkelMeshTransforms=::NewEnumerator0
    Skeletal Mesh=$User.SkelMesh
    Socket Sampling Mode=::NewEnumerator5
    Use Sampled Mesh Attribute as Ribbon ID=false
   ) {enabled=false}
   &self:Update_MeshSocketOffsets(
    PosOffset:Vector3f=<-15 2 0>
    SM_SampledRotation:Quat4f=$Particles.SkeletalMeshLocation.SampledRotation
    SpriteRotation:NiagaraFloat=&Multiply_Float(A=2, B=$Engine.Time)
   ) {enabled=false, module=Update_MeshSocketOffsets}
   &SolveForcesAndVelocity(Clamp Velocity=false, Limit Acceleration=false)
  ]
  rendererClasses [&NiagaraSpriteRendererProperties]
  renderers [{rendererClass=&NiagaraSpriteRendererProperties}]
 }
 TieDye_Base {
  emitterUpdateScript [
   &EmitterState()
   &SpawnBurst_Instantaneous(Spawn Count=1, Use Spawn Probability=false)
  ]
  particleSpawnScript [&InitializeParticle(
    Color Mode=::NewEnumerator0
    Lifetime=15
    Position Offset=<0 0 125>
    Sprite Rotation Mode=::NewEnumerator0
    Sprite Size Mode=::NewEnumerator3
    Uniform Sprite Size=50
    UsePositionOffset=false
    Write Material Random=true
   )]
  particleUpdateScript [
   &ParticleState()
   &SkeletalMeshLocation(
    "Interpolate Mesh Sampling (Update)"=true
    MaskSpawnGroup=false
    Mesh Sampling Type=::NewEnumerator5
    Sampled Position Offset Coordinate Space:ENiagaraSkelMeshTransforms=::NewEnumerator0
    Skeletal Mesh=$User.SkelMesh
    Socket Sampling Mode=::NewEnumerator5
    Use Sampled Mesh Attribute as Ribbon ID=false
    Velocity Sampling=::NewEnumerator0
    VelocitySpeedLimit=false
   )
   &self:Update_MeshSocketOffsets(
    PosOffset:Vector3f=<3 0 0>
    SM_SampledRotation:Quat4f=$Particles.SkeletalMeshLocation.SampledRotation
    SpriteRotation:NiagaraFloat=&Multiply_Float(A=2, B=$Engine.Time)
   ) {module=Update_MeshSocketOffsets}
   /SetVariables(Particles.Halo_SpriteSize:Vector2f=&Multiply_Vector2D_ByFloat(Float=0.9, Vector2D=$Particles.SpriteSize)) {isSetParameters=true}
   &self:SetDynamicMaterialParms(ColorA:LinearColor=$User.ColorA, ColorB:LinearColor=$User.ColorB, ColorC:LinearColor=$User.ColorC) {module=SetDynamicMaterialParms}
  ]
  rendererClasses [&NiagaraSpriteRendererProperties, &NiagaraMeshRendererProperties]
  renderers [
   {rendererClass=&NiagaraSpriteRendererProperties}
   {rendererClass=&NiagaraMeshRendererProperties, rendererIndex=1}
   {rendererClass=&NiagaraMeshRendererProperties, rendererIndex=2}
   {rendererClass=&NiagaraSpriteRendererProperties, rendererIndex=3}
  ]
 }
}
summary {
 systemName=N_F_MED_TieDyeFashion
 emitters [
  {emitterName=TieDye_Base, rendererClasses [&NiagaraSpriteRendererProperties, &NiagaraMeshRendererProperties]}
  {bEnabled=false, emitterName=Empty, rendererClasses [&NiagaraSpriteRendererProperties]}
 ]
 userVariables [
  User.ColorA {type=LinearColor, description="", defaultValue {struct=LinearColor, value=<rgba 0.0823 0.5981 1 1>}}
  User.ColorB {type=LinearColor, description="", defaultValue {struct=LinearColor, value=<rgba 0 0.1011 1 1>}}
  User.ColorC {type=LinearColor, description="", defaultValue {struct=LinearColor, value=<rgba 1 0.0828 0.0729 1>}}
  User.SkelMesh {
   type=NiagaraDataInterfaceSkeletalMesh
   defaultValue {
    struct=NiagaraExt_VariableValue_DataInterface
    value {dataInterface=&self:NiagaraDataInterfaceSkeletalMesh_0}
   }
  }
 ]
}
